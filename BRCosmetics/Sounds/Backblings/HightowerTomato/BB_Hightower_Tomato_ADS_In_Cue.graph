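# Sound cue for the Hightower Tomato back bling cosmetic, named as its aim-down-sights-in sound.
# A Random node picks one of four equally weighted wave variants, then a Modulator applies slight pitch variation.
# It plays at reduced volume in the weapon foley sound class, borrowing weapon reload attenuation and using the reactive back bling audio property sheet.
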SoundCue /BRCosmetics/Sounds/Backblings/HightowerTomato/BB_Hightower_Tomato_ADS_In_Cue
root=Modulator_1
nodes [
 {id=Modulator_1, children [Random_2], properties {pitchMax=1.02, pitchMin=0.98}}
 {id=Random_2, children [WavePlayer_1,WavePlayer_7,WavePlayer_8,WavePlayer_9], properties {weights [1,1,1,1]}}
 {id=WavePlayer_1, properties {soundWaveAssetPtr=&/BB_Hightower_Tomato_ADS_In_01}}
 {id=WavePlayer_7, properties {soundWaveAssetPtr=&/BB_Hightower_Tomato_ADS_In_02}}
 {id=WavePlayer_8, properties {soundWaveAssetPtr=&/BB_Hightower_Tomato_ADS_In_03}}
 {id=WavePlayer_9, properties {soundWaveAssetPtr=&/BB_Hightower_Tomato_ADS_In_04}}
]
properties {
 attenuationSettings=&Fort_Weapon_Reload_Att
 audioPropertiesSheet=/BRCosmetics/Sounds/Settings/PropertySheets/Reactive/APS_Cosmetic_Reactive_Backbling
 maxDistance=2250
 soundClassObject=&Fort_WeaponFoley
 volumeMultiplier=0.2
}
