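# Sound cue for the Hightower Tomato back bling cosmetic, named for an aim-down-sights exit sound.
# A random node picks one of four equally weighted wave variations and feeds it through a modulator.
# It plays quietly as weapon foley, using weapon reload attenuation and the reactive back bling audio property sheet.
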SoundCue /BRCosmetics/Sounds/Backblings/HightowerTomato/BB_Hightower_Tomato_ADS_Out_Cue
root=Modulator_0
nodes [
 {id=Modulator_0, children [Random_2]}
 {id=Random_2, children [WavePlayer_1, WavePlayer_8, WavePlayer_9, WavePlayer_10], properties {weights [1,1,1,1]}}
 {id=WavePlayer_1, properties {soundWaveAssetPtr=&/BB_Hightower_Tomato_ADS_Out_01}}
 {id=WavePlayer_10, properties {soundWaveAssetPtr=&/BB_Hightower_Tomato_ADS_Out_04}}
 {id=WavePlayer_8, properties {soundWaveAssetPtr=&/BB_Hightower_Tomato_ADS_Out_02}}
 {id=WavePlayer_9, properties {soundWaveAssetPtr=&/BB_Hightower_Tomato_ADS_Out_03}}
]
properties {
 attenuationSettings=&Fort_Weapon_Reload_Att
 audioPropertiesSheet=/BRCosmetics/Sounds/Settings/PropertySheets/Reactive/APS_Cosmetic_Reactive_Backbling
 maxDistance=2250
 soundClassObject=&Fort_WeaponFoley
 volumeMultiplier=0.125
}
